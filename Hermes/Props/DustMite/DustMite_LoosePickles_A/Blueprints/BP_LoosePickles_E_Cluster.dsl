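; Blueprint for the LoosePickles E cluster prop in the DustMite prop set, built on a parent class that supplies a static mesh component.
; Its construction script runs the parent's script, then hides that mesh without propagating to children and turns off its collision.
; BeginPlay, ActorBeginOverlap and Tick are empty, so this graph adds no runtime logic of its own.
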(event EventBeginPlay)

(event Collision|EventActorBeginOverlap
  (params (OtherActor :type Actor%20Object%20Reference)))

(event EventTick
  (params (DeltaSeconds :type Float%20%28single-precision%29)))

(fn ConstructionScript
  (bind _staticmeshcomponent (Variables|Debug|GetStaticMeshComponent))
  (|CallParentFunction)
  (Rendering|SetVisibility _staticmeshcomponent false false)
  (Collision|SetCollisionEnabled _staticmeshcomponent "NoCollision"))
